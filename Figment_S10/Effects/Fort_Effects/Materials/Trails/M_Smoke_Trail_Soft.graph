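# Translucent soft smoke trail material for Niagara ribbons, beam trails and particle sprites, colored by particle color and an optional color texture.
# A grayscale trail texture pans along UV1; dynamic parameters drive its erosion, pan speed, emissive boost and master alpha.
# Opacity combines the eroded mask with a both-ends U gradient, camera and scene depth fades, and view-angle fade.
# ES2 and ES3_1 skip those three fades and MasterAlpha.
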Material /Figment_S10/Effects/Fort_Effects/Materials/Trails/M_Smoke_Trail_Soft
graph {
 properties {
  bUsedWithBeamTrails=true
  bUsedWithNiagaraRibbons=true
  bUsedWithParticleSprites=true
  blendMode=BLEND_Translucent
  pixelDepthOffsetMode=PDOM_Legacy
  usedShadingModels=Default Lit
 }
 nodes [
  n1 = ParticleColor()
  n2 = DynamicParameter() {defaultValue=<rgba 1 0 1 1>, paramNames [Erode, EmissiveAmount, PanSpeed, MasterAlpha]}
  n3 = Panner(Coordinate=TextureCoordinate() {coordinateIndex=1}, Time=Multiply(A=@n2.PanSpeed, B=Time())) {speedX=0.6}
  n4 = TextureSampleParameter2D(UVs=@n3) {parameterName=TrailTexture_BChannel, samplerType=SAMPLERTYPE_Grayscale, texture=&T_Smoke_Trail_Soft}
  n5 = Add(A=@n2.Erode, B=@n4.G)
  n6 = Clamp(None=Multiply(A=LinearInterpolate(Alpha=@n5) {constA=-3}, B=@n4.G))
  n7 = Multiply(A=@n1.A, B=@n6)
  n8 = ScalarParameter() {defaultValue=350, parameterName=Fade Length}
  n9 = ScalarParameter() {defaultValue=24, parameterName=Fade Offset}
  n10 = MaterialFunctionCall() {materialFunction=&LinearGradient}
  n11 = Multiply(A=OneMinus(None=@n10.UGradient), B=@n10.UGradient)
  n12 = Clamp(None=MaterialFunctionCall(In=Multiply(A=@n11) {constB=8}, Contrast=ScalarParameter() {defaultValue=0.1, parameterName=TailGradientFallOff}) {materialFunction=&CheapContrast}.Result)
  n13 = Multiply(
   A=Multiply(
    A=@n7
    B=MaterialFunctionCall(Fade Length=@n8, Fade Offset=@n9) {materialFunction=&CameraDepthFade}.Result
   )
   B=@n12
  )
  n14 = ScalarParameter() {parameterName=EdgeFadeSensitivity}
  n15 = Multiply(
   A=Multiply(
    A=DepthFade(Opacity=@n13) {fadeDistanceDefault=40}
    B=MaterialFunctionCall("FadeSensitivity (optional)"=@n14) {materialFunction=&FadeBasedOnViewAngle}.Result
   )
   B=DynamicParameter() {defaultValue=<rgba 1 0 1 1>, paramNames [Erode, EmissiveAmount, PanSpeed, MasterAlpha]}.MasterAlpha
  )
  n16 = Multiply(A=@n7, B=@n12)
  FeatureLevelSwitch(Default=@n15, ES2=@n16, ES3_1=@n16, SM4_REMOVED=@n15, SM5=@n15, SM6=@n15)
  n17 = Multiply(
   A=@n1.RGB
   B=StaticSwitchParameter(
    True=TextureSampleParameter2D(UVs=@n3) {parameterName=ColorTexture, texture=/Game/Effects/Fort_Effects/Textures/Smoke/T_Smoke_Trail_Soft_Chromatic}.RGB
    False=Constant() {r=1}
   ) {parameterName=UseColorTexture}
  )
  n18 = DynamicParameter() {defaultValue=<rgba 1 0 1 1>, paramNames [Erode, EmissiveAmount, PanSpeed, MasterAlpha]}
  MaterialFunctionCall(LDR Input=Clamp(None=@n17), HDR Tint=Multiply(A=@n17, B=Multiply(A=@n18.EmissiveAmount, B=@n6))) {materialFunction=&DeriveHDRfromLDR}
  Reroute(None=@n1.RGB)
 ]
}
